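# Physics asset for the Jungle mic stand instrument cosmetic in SparksCosmetics, giving its leaves limited physical movement.
# Six bunch bones carry tiny capsule bodies, and each of their seventeen leaf bones has a fitted capsule.
# Every leaf is constrained only to its own bunch, with small per-leaf swing and twist limits.
# One leaf, bunch_01_leaf_D, has swing1 and twist locked, leaving it a single swing axis.
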PhysicsAsset /SparksCosmetics/Instrument/MicStand/Jungle/Meshes/Jungle_MicStand_Physics
bodies {
 bunch_01 [{shapeType=Capsule, length=0.5, radius=0.5, shapeName=bunch_01_capsule}]
 bunch_01_leaf_A [{
   shapeType=Capsule
   length=2.8877
   radius=2.3362
   shapeName=bunch_01_leaf_A_capsule
   center=<3.8328 0.416 0.5127>
   rotation=<rot -89.3698 0.0152 -0.0152>
  }]
 bunch_01_leaf_B [{
   shapeType=Capsule
   length=3.4306
   radius=2.0694
   shapeName=bunch_01_leaf_B_capsule
   center=<4.0747 -0.0737 -1.0475>
   rotation=<rot -82 180 -180>
  }]
 bunch_01_leaf_C [{
   shapeType=Capsule
   length=4.4577
   radius=1.5423
   shapeName=bunch_01_leaf_C_capsule
   center=<3.2092 0 0.4241>
   rotation=<rot -85.1314 -179.9991 179.9992>
  }]
 bunch_01_leaf_D [{
   shapeType=Capsule
   length=2.7517
   radius=2.2483
   shapeName=bunch_01_leaf_D_capsule
   center=<1.1478 0.1869 -2.4059>
   rotation=<rot -79.2693 165.6281 96.596>
  }]
 bunch_01_leaf_E [{
   shapeType=Capsule
   length=4
   radius=1.5
   shapeName=bunch_01_leaf_E_capsule
   center=<3.4404 0 -0.3449>
   rotation=<rot -84.2744 179.9989 -179.9988>
  }]
 bunch_01_leaf_F [{
   shapeType=Capsule
   length=3.4979
   radius=2.0021
   shapeName=bunch_01_leaf_F_capsule
   center=<3.2085 -0.5078 0.9815>
   rotation=<rot -79.6642 -179.9999 179.9998>
  }]
 bunch_02 [{shapeType=Capsule, length=0.5, radius=0.5, shapeName=bunch_02_capsule}]
 bunch_02_leaf_A [{shapeType=Capsule, length=3.6624, radius=1.2597, shapeName=bunch_02_leaf_A_capsule, center=<3 0 0.5>, rotation=<rot 0 84.8667 -89.529>}]
 bunch_02_leaf_B [{
   shapeType=Capsule
   length=4.341
   radius=1.3717
   shapeName=bunch_02_leaf_B_capsule
   center=<3.9266 -0.0025 -0.3469>
   rotation=<rot 0 -88.0803 -78.0512>
  }]
 bunch_02_leaf_C [{shapeType=Capsule, length=2.4904, radius=2, shapeName=bunch_02_leaf_C_capsule, center=<3 0 -1>, rotation=<rot 0 -91.9478 -80.5704>}]
 bunch_03 [{shapeType=Capsule, length=0.5, radius=0.5, shapeName=bunch_03_capsule, rotation=<rot -0.02 0 -75.8203>}]
 bunch_03_leaf_A [{
   shapeType=Capsule
   length=3.1404
   radius=1.7596
   shapeName=bunch_03_leaf_A_capsule
   center=<2.8298 0.1174 -0.0904>
   rotation=<rot 0 -90.0405 -89.0212>
  }]
 bunch_03_leaf_B [{
   shapeType=Capsule
   length=3.6
   radius=1.3
   shapeName=bunch_03_leaf_B_capsule
   center=<3.0744 0.0838 -0.2219>
   rotation=<rot 0 -87.9827 -86.8662>
  }]
 bunch_03_leaf_C [{
   shapeType=Capsule
   length=4.9051
   radius=1.3
   shapeName=bunch_03_leaf_C_capsule
   center=<3.8433 0.0389 -0.3278>
   rotation=<rot 0 -89.1849 -83.8389>
  }]
 bunch_04 [{shapeType=Capsule, length=0.5, radius=0.5, shapeName=bunch_04_capsule, rotation=<rot -48.3737 0.0002 12.6045>}]
 bunch_04_leaf_A [{
   shapeType=Capsule
   length=4.2872
   radius=1.7128
   shapeName=bunch_04_leaf_A_capsule
   center=<4.3442 -0.6395 -0.3196>
   rotation=<rot 0 -98.5808 -84.3231>
  }]
 bunch_04_leaf_B [{
   shapeType=Capsule
   length=4.313
   radius=1.687
   shapeName=bunch_04_leaf_B_capsule
   center=<4.3269 -0.1949 -0.1935>
   rotation=<rot 0 -93.729 -84.6843>
  }]
 bunch_04_leaf_C [{shapeType=Capsule, length=4, radius=2, shapeName=bunch_04_leaf_C_capsule, center=<4.1975 0.0053 -0.1951>, rotation=<rot 0 -90.0609 -86.2852>}]
 bunch_05 [{shapeType=Capsule, length=0.5, radius=0.5, shapeName=bunch_05_capsule, rotation=<rot -80.6658 0.0004 -0.0006>}]
 bunch_05_leaf_A [{
   shapeType=Capsule
   length=1.7907
   radius=4.2093
   shapeName=bunch_05_leaf_A_capsule
   center=<4.1907 -0.3714 -0.3893>
   rotation=<rot 8.7222 -96.0983 -85.6065>
  }]
 bunch_06 [{shapeType=Capsule, length=0.5, radius=0.5, shapeName=bunch_06_capsule, rotation=<rot -6.4397 4.415 -38.5716>}]
 bunch_06_leaf_A [{
   shapeType=Capsule
   length=4
   radius=3.5155
   shapeName=bunch_06_leaf_A_capsule
   center=<4.4233 1.6985 -0.858>
   rotation=<rot -74.0342 179.7042 -0.0002>
  }]
}
constraints [
 {bone1Name=bunch_01_leaf_A, bone2Name=bunch_01, swing1LimitDegrees=15, swing2LimitDegrees=30, twistLimitDegrees=5}
 {bone1Name=bunch_01_leaf_B, bone2Name=bunch_01, swing1LimitDegrees=15, swing2LimitDegrees=20, twistLimitDegrees=5}
 {bone1Name=bunch_01_leaf_C, bone2Name=bunch_01, swing1LimitDegrees=15, swing2LimitDegrees=20, twistLimitDegrees=5}
 {
  bone1Name=bunch_01_leaf_D
  bone2Name=bunch_01
  swing1LimitDegrees=8
  swing1Motion=Locked
  swing2LimitDegrees=15
  twistLimitDegrees=0
  twistMotion=Locked
 }
 {bone1Name=bunch_01_leaf_E, bone2Name=bunch_01, swing1LimitDegrees=15, swing2LimitDegrees=10, twistLimitDegrees=5}
 {bone1Name=bunch_01_leaf_F, bone2Name=bunch_01, swing1LimitDegrees=15, swing2LimitDegrees=20, twistLimitDegrees=5}
 {bone1Name=bunch_03_leaf_A, bone2Name=bunch_03, swing1LimitDegrees=10, swing2LimitDegrees=15, twistLimitDegrees=10}
 {bone1Name=bunch_03_leaf_B, bone2Name=bunch_03, swing1LimitDegrees=10, swing2LimitDegrees=15, twistLimitDegrees=10}
 {bone1Name=bunch_03_leaf_C, bone2Name=bunch_03, swing1LimitDegrees=10, swing2LimitDegrees=15, twistLimitDegrees=10}
 {bone1Name=bunch_02_leaf_A, bone2Name=bunch_02, swing1LimitDegrees=15, swing2LimitDegrees=15, twistLimitDegrees=20}
 {bone1Name=bunch_02_leaf_B, bone2Name=bunch_02, swing1LimitDegrees=15, swing2LimitDegrees=15, twistLimitDegrees=20}
 {bone1Name=bunch_02_leaf_C, bone2Name=bunch_02, swing1LimitDegrees=10, swing2LimitDegrees=15, twistLimitDegrees=20}
 {bone1Name=bunch_04_leaf_A, bone2Name=bunch_04, swing1LimitDegrees=15, swing2LimitDegrees=13, twistLimitDegrees=8}
 {bone1Name=bunch_04_leaf_B, bone2Name=bunch_04, swing1LimitDegrees=15, swing2LimitDegrees=13, twistLimitDegrees=8}
 {bone1Name=bunch_04_leaf_C, bone2Name=bunch_04, swing1LimitDegrees=10, swing2LimitDegrees=15, twistLimitDegrees=8}
 {bone1Name=bunch_05_leaf_A, bone2Name=bunch_05, swing1LimitDegrees=15, swing2LimitDegrees=10, twistLimitDegrees=3}
 {bone1Name=bunch_06_leaf_A, bone2Name=bunch_06, swing1LimitDegrees=3, swing2LimitDegrees=8, twistLimitDegrees=2}
]
