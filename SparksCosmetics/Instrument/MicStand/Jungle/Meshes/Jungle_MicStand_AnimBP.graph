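# Animation Blueprint for the Jungle mic stand cosmetic that adds physics motion and procedural leaf sway.
# A RigidBody node simulates in root bone space with GravityOverride as external force, switched by skydiving and parachute state.
# Its pose feeds chained additive bone rotations on leaves, driven by real-time sine waves.
# Wave amplitudes re-randomize every update; leaves share one yaw shake and mix pitch waves, while Wave 06 is unused.
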AnimBlueprint /SparksCosmetics/Instrument/MicStand/Jungle/Meshes/Jungle_MicStand_AnimBP
target_skeleton=&/Jungle_MicStand_Skeleton
graphs [
 AnimGraph {nodes [
   Component To Local()
   n1 = RigidBody(ComponentPose=Local To Component(LocalPose=Input Pose()), ExternalForce=Get GravityOverride()) {
    bOverrideWorldGravity=true
    overridePhysicsAsset=&/Jungle_MicStand_Physics
    simulationSpace=BaseBoneSpace
    alphaBoolBlend {blendOption=Linear}
    baseBoneRef {boneName=root}
    componentAppliedLinearAccClamp=<3000 3000 3000>
    componentLinearAccScale=<0.4 0.3 0.3>
    componentLinearVelScale=<0.4 0.3 0.3>
    overrideWorldGravity {z=-980}
    simSpaceSettings {worldAlpha=1, externalLinearVelocity {y=500, z=500}}
   }
   n2 = "Transform (Modify) Bone - Bone: bunch_01_leaf_A"!(
    ComponentPose=@n1
    bAlphaBoolEnabled:bool=True
    Translation="0,0,0"
    Rotation_Roll:float=0
    Rotation_Pitch:float=Get wave_03() -> (wave_03:double)
    Rotation_Yaw:float=Get wave_shake() -> (wave_shake:double)
    Scale="1,1,1"
   ) -> (Pose) {
    alphaInputType=Bool
    rotationMode=BMM_Additive
    rotationSpace=BCS_ParentBoneSpace
    alphaBoolBlend {blendOption=Linear}
    boneToModify {boneName=bunch_01_leaf_A}
   }
   n3 = "Transform (Modify) Bone - Bone: bunch_01_leaf_B"!(
    ComponentPose=@n2
    bAlphaBoolEnabled:bool=True
    Translation="0,0,0"
    Rotation_Roll:float=0
    Rotation_Pitch:float=Get wave_01() -> (wave_01:double)
    Rotation_Yaw:float=Get wave_shake()
    Scale="1,1,1"
   ) -> (Pose) {
    alphaInputType=Bool
    rotationMode=BMM_Additive
    rotationSpace=BCS_ParentBoneSpace
    alphaBoolBlend {blendOption=Linear}
    boneToModify {boneName=bunch_01_leaf_B}
   }
   n4 = "Transform (Modify) Bone - Bone: bunch_01_leaf_C"!(
    ComponentPose=@n3
    bAlphaBoolEnabled:bool=True
    Translation="0,0,0"
    Rotation_Roll:float=0
    Rotation_Pitch:float=Get Wave 04() -> (Wave 04:double)
    Rotation_Yaw:float=Get wave_shake()
    Scale="1,1,1"
   ) -> (Pose) {
    alphaInputType=Bool
    rotationMode=BMM_Additive
    rotationSpace=BCS_ParentBoneSpace
    alphaBoolBlend {blendOption=Linear}
    boneToModify {boneName=bunch_01_leaf_C}
   }
   n5 = "Transform (Modify) Bone - Bone: bunch_01_leaf_D"!(
    ComponentPose=@n4
    bAlphaBoolEnabled:bool=True
    Translation="0,0,0"
    Rotation_Roll:float=0
    Rotation_Pitch:float=Get Wave 05() -> (Wave 05:double)
    Rotation_Yaw:float=Get wave_shake()
    Scale="1,1,1"
   ) -> (Pose) {
    alphaInputType=Bool
    rotationMode=BMM_Additive
    rotationSpace=BCS_ParentBoneSpace
    alphaBoolBlend {blendOption=Linear}
    boneToModify {boneName=bunch_01_leaf_D}
   }
   n6 = "Transform (Modify) Bone - Bone: bunch_01_leaf_E"!(
    ComponentPose=@n5
    bAlphaBoolEnabled:bool=True
    Translation="0,0,0"
    Rotation_Roll:float=0
    Rotation_Pitch:float=Get wave_03()
    Rotation_Yaw:float=Get wave_shake()
    Scale="1,1,1"
   ) -> (Pose) {
    alphaInputType=Bool
    rotationMode=BMM_Additive
    rotationSpace=BCS_ParentBoneSpace
    alphaBoolBlend {blendOption=Linear}
    boneToModify {boneName=bunch_01_leaf_E}
   }
   n7 = "Transform (Modify) Bone - Bone: bunch_01_leaf_F"!(
    ComponentPose=@n6
    bAlphaBoolEnabled:bool=True
    Translation="0,0,0"
    Rotation_Roll:float=0
    Rotation_Pitch:float=Get Wave 04()
    Rotation_Yaw:float=Get wave_shake()
    Scale="1,1,1"
   ) -> (Pose) {
    alphaInputType=Bool
    rotationMode=BMM_Additive
    rotationSpace=BCS_ParentBoneSpace
    alphaBoolBlend {blendOption=Linear}
    boneToModify {boneName=bunch_01_leaf_F}
   }
   n8 = "Transform (Modify) Bone - Bone: bunch_02_leaf_A"!(
    ComponentPose=@n7
    bAlphaBoolEnabled:bool=True
    Translation="0,0,0"
    Rotation_Roll:float=0
    Rotation_Pitch:float=Get wave_01()
    Rotation_Yaw:float=Get wave_shake()
    Scale="1,1,1"
   ) -> (Pose) {
    alphaInputType=Bool
    rotationMode=BMM_Additive
    rotationSpace=BCS_ParentBoneSpace
    alphaBoolBlend {blendOption=Linear}
    boneToModify {boneName=bunch_02_leaf_A}
   }
   n9 = "Transform (Modify) Bone - Bone: bunch_02_leaf_B"!(
    ComponentPose=@n8
    bAlphaBoolEnabled:bool=True
    Translation="0,0,0"
    Rotation_Roll:float=0
    Rotation_Pitch:float=Get Wave 05()
    Rotation_Yaw:float=Get wave_shake()
    Scale="1,1,1"
   ) -> (Pose) {
    alphaInputType=Bool
    rotationMode=BMM_Additive
    rotationSpace=BCS_ParentBoneSpace
    alphaBoolBlend {blendOption=Linear}
    boneToModify {boneName=bunch_02_leaf_B}
   }
   n10 = "Transform (Modify) Bone - Bone: bunch_02_leaf_C"!(
    ComponentPose=@n9
    bAlphaBoolEnabled:bool=True
    Translation="0,0,0"
    Rotation_Roll:float=0
    Rotation_Pitch:float=Get wave_01()
    Rotation_Yaw:float=Get wave_shake()
    Scale="1,1,1"
   ) -> (Pose) {
    alphaInputType=Bool
    rotationMode=BMM_Additive
    rotationSpace=BCS_ParentBoneSpace
    alphaBoolBlend {blendOption=Linear}
    boneToModify {boneName=bunch_02_leaf_C}
   }
   n11 = "Transform (Modify) Bone - Bone: bunch_03_leaf_A"!(
    ComponentPose=@n10
    bAlphaBoolEnabled:bool=True
    Translation="0,0,0"
    Rotation_Roll:float=0
    Rotation_Pitch:float=Get wave_01()
    Rotation_Yaw:float=Get wave_shake()
    Scale="1,1,1"
   ) -> (Pose) {
    alphaInputType=Bool
    rotationMode=BMM_Additive
    rotationSpace=BCS_ParentBoneSpace
    alphaBoolBlend {blendOption=Linear}
    boneToModify {boneName=bunch_03_leaf_A}
   }
   n12 = "Transform (Modify) Bone - Bone: bunch_03_leaf_B"!(
    ComponentPose=@n11
    bAlphaBoolEnabled:bool=True
    Translation="0,0,0"
    Rotation_Roll:float=0
    Rotation_Pitch:float=Get Wave 04()
    Rotation_Yaw:float=Get wave_shake()
    Scale="1,1,1"
   ) -> (Pose) {
    alphaInputType=Bool
    rotationMode=BMM_Additive
    rotationSpace=BCS_ParentBoneSpace
    alphaBoolBlend {blendOption=Linear}
    boneToModify {boneName=bunch_03_leaf_B}
   }
   n13 = "Transform (Modify) Bone - Bone: bunch_03_leaf_C"!(
    ComponentPose=@n12
    bAlphaBoolEnabled:bool=True
    Translation="0,0,0"
    Rotation_Roll:float=0
    Rotation_Pitch:float=Get wave_02() -> (wave_02:double)
    Rotation_Yaw:float=Get wave_shake()
    Scale="1,1,1"
   ) -> (Pose) {
    alphaInputType=Bool
    rotationMode=BMM_Additive
    rotationSpace=BCS_ParentBoneSpace
    alphaBoolBlend {blendOption=Linear}
    boneToModify {boneName=bunch_03_leaf_C}
   }
   n14 = "Transform (Modify) Bone - Bone: bunch_04_leaf_A"!(
    ComponentPose=@n13
    bAlphaBoolEnabled:bool=True
    Translation="0,0,0"
    Rotation_Roll:float=0
    Rotation_Pitch:float=Get wave_01()
    Rotation_Yaw:float=Get wave_shake()
    Scale="1,1,1"
   ) -> (Pose) {
    alphaInputType=Bool
    rotationMode=BMM_Additive
    rotationSpace=BCS_ParentBoneSpace
    alphaBoolBlend {blendOption=Linear}
    boneToModify {boneName=bunch_04_leaf_A}
   }
   n15 = "Transform (Modify) Bone - Bone: bunch_04_leaf_B"!(
    ComponentPose=@n14
    bAlphaBoolEnabled:bool=True
    Translation="0,0,0"
    Rotation_Roll:float=0
    Rotation_Pitch:float=Get Wave 05()
    Rotation_Yaw:float=Get wave_shake()
    Scale="1,1,1"
   ) -> (Pose) {
    alphaInputType=Bool
    rotationMode=BMM_Additive
    rotationSpace=BCS_ParentBoneSpace
    alphaBoolBlend {blendOption=Linear}
    boneToModify {boneName=bunch_04_leaf_B}
   }
   n16 = "Transform (Modify) Bone - Bone: bunch_04_leaf_C"!(
    ComponentPose=@n15
    bAlphaBoolEnabled:bool=True
    Translation="0,0,0"
    Rotation_Roll:float=0
    Rotation_Pitch:float=Get wave_01()
    Rotation_Yaw:float=Get wave_shake()
    Scale="1,1,1"
   ) -> (Pose) {
    alphaInputType=Bool
    rotationMode=BMM_Additive
    rotationSpace=BCS_ParentBoneSpace
    alphaBoolBlend {blendOption=Linear}
    boneToModify {boneName=bunch_04_leaf_C}
   }
   n17 = "Transform (Modify) Bone - Bone: bunch_05_leaf_A"!(
    ComponentPose=@n16
    bAlphaBoolEnabled:bool=True
    Translation="0,0,0"
    Rotation_Roll:float=0
    Rotation_Pitch:float=Get wave_01()
    Rotation_Yaw:float=Get wave_shake()
    Scale="1,1,1"
   ) -> (Pose) {
    alphaInputType=Bool
    rotationMode=BMM_Additive
    rotationSpace=BCS_ParentBoneSpace
    alphaBoolBlend {blendOption=Linear}
    boneToModify {boneName=bunch_05_leaf_A}
   }
   n18 = "Transform (Modify) Bone - Bone: bunch_06_leaf_A"!(
    ComponentPose=@n17
    bAlphaBoolEnabled:bool=True
    Translation="0,0,0"
    Rotation_Roll:float=0
    Rotation_Pitch:float=Get Wave 04()
    Rotation_Yaw:float=Get wave_shake()
    Scale="1,1,1"
   ) -> (Pose) {
    alphaInputType=Bool
    rotationMode=BMM_Additive
    rotationSpace=BCS_ParentBoneSpace
    alphaBoolBlend {blendOption=Linear}
    boneToModify {boneName=bunch_06_leaf_A}
   }
   Output Pose(Result=Component To Local(ComponentPose=@n18))
  ]}
 EventGraph {nodes [
   Event BlueprintUpdateAnimation()
   > n1 = Sequence()
   > n2 = Branch(Condition=Get bIsSkyDiving())
   > n3 = Branch(Condition=Get bIsParachuteOpened())
   > Set GravityOverride(GravityOverride="0,0,-400")
   Set GravityOverride(execute=@n3.else, GravityOverride="0,0,-10000")
   Set GravityOverride(execute=@n2.else, GravityOverride="0,0,-980")
   n4 = GetRealTimeSeconds()
   n5 = float * float(A=@n4, B=8)
   Set wave_01(execute=@n1.then_1, wave_01:double=float * float(A="Sin (Radians)"(A=@n5), B=RandomFloatInRange(Min=0.5, Max=0.75))) -> (Output_Get:double)
   n6 = float * float(A=@n4, B=10)
   > Set wave_02(wave_02:double=float * float(A="Sin (Radians)"(A=@n6), B=RandomFloatInRange(Min=1, Max=1.5))) -> (Output_Get:double)
   n7 = float * float(A=@n4, B=12)
   > Set wave_03(wave_03:double=float * float(A="Sin (Radians)"(A=@n7), B=RandomFloatInRange(Min=2, Max=3))) -> (Output_Get:double)
   n8 = float * float(A=@n4, B=2)
   > Set wave_shake(wave_shake:double=float * float(A="Sin (Radians)"(A=@n8), B=RandomFloatInRange(Min=0.1, Max=0.5))) -> (Output_Get:double)
   n9 = float * float(A=@n4, B=8.2)
   > Set Wave 04(Wave 04:double=float * float(A="Sin (Radians)"(A=@n9), B=RandomFloatInRange(Min=-0.5, Max=-0.75))) -> (Output_Get:double)
   n10 = float * float(A=@n4, B=10.2)
   > Set Wave 05(Wave 05:double=float * float(A="Sin (Radians)"(A=@n10), B=RandomFloatInRange(Min=-1, Max=-1.5))) -> (Output_Get:double)
   n11 = float * float(A=@n4, B=12.2)
   > Set Wave 06(Wave 06:double=float * float(A="Sin (Radians)"(A=@n11), B=RandomFloatInRange(Min=-2, Max=-3))) -> (Output_Get:double)
   TryGetPawnOwner()
  ]}
]
variables [
 GravityOverride {type=Vector}
 Wave 04 {type=number}
 Wave 05 {type=number}
 Wave 06 {type=number}
 wave_01 {type=number}
 wave_02 {type=number}
 wave_03 {type=number}
 wave_shake {type=number}
]
